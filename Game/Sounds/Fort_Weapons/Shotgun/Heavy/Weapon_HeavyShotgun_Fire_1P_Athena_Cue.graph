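# First-person firing sound cue for the Heavy Shotgun in Athena.
# A Random node picks one of three close shotgun fire waves, feeding a Modulator for slight per-shot variation.
# The Modulator enters a Mixer whose other two inputs are empty, so only the close layer plays.
# First-person weapon attenuation, a ranged-fire sound class and a gunfire concurrency asset govern playback.
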SoundCue /Game/Sounds/Fort_Weapons/Shotgun/Heavy/Weapon_HeavyShotgun_Fire_1P_Athena_Cue
root=Mixer_0
nodes [
 {id=Mixer_0, children [Modulator_0,null,null], properties {inputVolume [0.8,0.5,0.3]}}
 {id=Modulator_0, children [Random_0], properties {pitchMin=1}}
 {id=Random_0, children [WavePlayer_0,WavePlayer_1,WavePlayer_2], properties {weights [1,1,1]}}
 {id=WavePlayer_0, properties {soundWaveAssetPtr=&/Shotgun_NewFire_Close_01}}
 {id=WavePlayer_1, properties {soundWaveAssetPtr=&/Shotgun_NewFire_Close_02}}
 {id=WavePlayer_2, properties {soundWaveAssetPtr=&/Shotgun_NewFire_Close_03}}
]
properties {
 attenuationSettings=&Fort_Weapon_Projectile_FirstPerson_Athena_Att
 maxDistance=2097152
 soundClassObject=&Fort_WeaponRanged_Fire_1P_PSM
 volumeMultiplier=1.2
 attenuationOverrides {falloffDistance=5000, attenuationShapeExtents {x=1000}}
 concurrencyOverrides {bLimitToOwner=true, maxCount=3, resolutionRule=StopOldest}
 concurrencySet [&Weapon_Gunfire_FirstPerson_Conc]
}
